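# Unlit translucent two-sided Niagara sprite material for the Armadillo Robot emote's glowing heart effect.
# Color blends a time-of-day glow with dark-to-light red through heart texture masks, whose UV scale pulses with a looping damped sine.
# Opacity combines a switchable light-cone or sphere-mask shape, distance and depth fades, and panning scanline fog, triplanar by default.
# Particle alpha or an Alpha parameter scales the final opacity.
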Material /FNGameplayCues/Animation/Game/MainPlayer/Emotes/ArmadilloRobot/FX/M_Armadillo_Robot_Emote_Heart
graph {
 properties {
  bUsedWithNiagaraSprites=true
  blendMode=BLEND_Translucent
  pixelDepthOffsetMode=PDOM_Legacy
  shadingModel=MSM_Unlit
  twoSided=true
  usedShadingModels=Unlit
 }
 nodes [
  n1 = ScalarParameter() {defaultValue=1, group=Color Adjustments, parameterName=Morning Multiplier}
  n2 = VectorParameter() {parameterName=GlowColor, defaultValue=<rgba 1 0.5328 0.124 1>}
  n3 = ScalarParameter() {defaultValue=1, group=Color Adjustments, parameterName=Day Multiplier}
  n4 = ScalarParameter() {defaultValue=1, group=Color Adjustments, parameterName=Evening Multiplier}
  n5 = ScalarParameter() {defaultValue=1, group=Color Adjustments, parameterName=Night Multiplier}
  n6 = Constant() {r=1}
  n7 = Divide(A=Fmod(A=Time() {period=1.5}, B=Constant() {r=1})) {constB=1.5}
  n8 = Add(A=Multiply(A=Power(Base=@n7) {constExponent=0.2}) {constB=0.5}) {constB=0.5}
  n9 = MaterialFunctionCall(Float2=Add(A=TextureCoordinate() {uTiling=2, vTiling=1.5}, B=Constant2Vector() {g=-0.5, r=-1})) {materialFunction=&BreakOutFloat2Components}
  n10 = Add(A=Multiply(A=@n7, B=Constant() {r=6.2881}) {constB=6.2831}, B=Multiply(A=Multiply(A=@n9.G) {constB=0.5}, B=Constant() {r=3}))
  n11 = Power(Base=OneMinus(None=@n7)) {constExponent=4}
  n12 = Multiply(A=Multiply(A=Add(A=@n6, B=@n8)) {constB=0.5}, B=Sine(None=Multiply(A=@n10, B=@n11))) {constB=0.5}
  n13 = TextureSample(UVs=MaterialFunctionCall(UVs=TextureCoordinate(), Texture Scale=LinearInterpolate(Alpha=@n12) {constA=0.93}) {materialFunction=&ScaleUVsByCenter}.UVs) {samplerType=SAMPLERTYPE_LinearColor, texture=&/T_ArmadilloRobotHeart_M}
  LinearInterpolate(
   A=MaterialFunctionCall(
    Color_Morning=Multiply(A=@n1, B=@n2.RGB) {constB=250}
    Color_Day=Multiply(A=@n2.RGB, B=@n3) {constB=250}
    Color_Evening=Multiply(A=@n4, B=@n2.RGB) {constB=250}
    Color_Night=Multiply(A=@n5, B=@n2.RGB) {constB=250}
   ) {materialFunction=&Color_Time_of_Day}.Result
   B=LinearInterpolate(
    A=VectorParameter() {parameterName=HeartColorDark, defaultValue=<rgba 0.1927 0 0.0018 1>}.RGB
    B=VectorParameter() {parameterName=HeartColorLight, defaultValue=<rgba 0.7448 0 0.0068 1>}.RGB
    Alpha=@n13.G
   )
   Alpha=@n13.R
  )
  n14 = ScalarParameter() {defaultValue=4, group=Fresnel, parameterName=Base Fresnel Exponent}
  n15 = StaticSwitchParameter(
   True=MaterialFunctionCall("FadeSensitivity (optional)"=@n14) {materialFunction=&FadeBasedOnViewAngle}.Result
   False=OneMinus(None=Fresnel(ExponentIn=@n14))
  ) {parameterName=UseViewAngleFade}
  n16 = ScalarParameter() {defaultValue=12, group=Fresnel, parameterName=Base Fresnel Power}
  n17 = MaterialFunctionCall(
   Position=WorldPosition().XYZ
   Fade_Start_Distance=ScalarParameter() {defaultValue=250, group=Fade, parameterName="Fade_Start_Distance (S)"}
   Fade_End_Distance=ScalarParameter() {defaultValue=100, group=Fade, parameterName="Fade_End_Distance (S)"}
  ) {materialFunction=&MF_Distance_SyncedFade_FX}
  n18 = Power(Base=ComponentMask(None=TextureCoordinate()) {g=true}, Exp=ScalarParameter() {defaultValue=1, group=Fade, parameterName=Vertical Fade Power})
  n19 = ScalarParameter() {defaultValue=2, group=Fade, parameterName=Vertical Fade Mult}
  n20 = Multiply(A=Multiply(A=Power(Base=@n15, Exp=@n16), B=@n17.Result), B=Saturate(None=Multiply(A=@n18, B=@n19)))
  n21 = MaterialFunctionCall("FadeSensitivity (optional)"=ScalarParameter() {group=Fade, parameterName=Angle Fade Sensitivity}) {materialFunction=&FadeBasedOnViewAngle}
  n22 = Power(
   Base=SphereMask(
    A=TextureCoordinate()
    B=Constant() {r=0.5}
    Radius=ScalarParameter() {defaultValue=0.5, group=End Symbol, parameterName=Radius}
    Hardness=ScalarParameter() {defaultValue=0.5, group=End Symbol, parameterName=Hardness}
   )
   Exp=ScalarParameter() {defaultValue=2, group=End Symbol, parameterName=SphereMask Power}
  )
  n23 = Multiply(A=@n17.Result, B=StaticSwitchParameter(True=Multiply(A=@n21.Result, B=@n22), False=@n22) {parameterName=Fade Based On Angle?})
  n24 = Multiply(
   A=DepthFade(FadeDistance=ScalarParameter() {parameterName=DepthFade}) {fadeDistanceDefault=25}
   B=Saturate(None=StaticSwitchParameter(True=@n20, False=@n23) {parameterName=IsLightCone?})
  )
  n25 = ScalarParameter() {defaultValue=1, parameterName=Final Power}
  n26 = Multiply(A=Clamp(None=Power(Base=@n24, Exp=@n25)), B=ScalarParameter() {defaultValue=0.1, parameterName=Final Multiplier})
  n27 = AppendVector(
   A=ScalarParameter() {defaultValue=0.005, group=Fog, parameterName=Fog Pan Speed X}
   B=ScalarParameter() {group=Fog, parameterName=Fog Pan Speed Y}
  )
  n28 = VectorParameter() {group=Fog, parameterName=FogTextureTiling, defaultValue=<rgba 2 2 2 1>}
  n29 = MaterialFunctionCall!(
   CameraVector
   Time Scale
   Panner Speed=@n27
   Texture=TextureObjectParameter() {group=Fog, parameterName=Fog Texture, texture=/Game/PvP/StrategicBuildingActors/Defensive/LaserFence/Assets/LaserScanline}
   Tiling=@n28.RGB
   Offset
   AxisFadeContrast
  ) {materialFunction=/Game/Weapons/WeaponSkins/MaterialFunction/TriplanarCamerVector_Panning}
  n30 = TextureCoordinate()
  n31 = ComponentMask(None=@n28.RGB) {g=true, r=true}
  n32 = Multiply(A=Time(), B=@n27)
  n33 = TextureSample(UVs=Add(A=Multiply(A=@n30, B=@n31), B=Frac(None=@n32))) {texture=&LaserScanline}
  n34 = LinearInterpolate(
   A=MaterialFunctionCall(
    In=StaticSwitchParameter(True=@n29.Result, False=@n33.G) {defaultValue=true, group=Fog, parameterName=UseTriPlanarFog?}
    Contrast=ScalarParameter() {parameterName=Cheap Contrast}
   ) {materialFunction=&CheapContrast}.Result
   Alpha=ScalarParameter() {defaultValue=0.5, group=Fog, parameterName=Fade Fog}
  )
  Multiply(
   A=LinearInterpolate(A=ParticleColor().A, B=ScalarParameter() {defaultValue=1, parameterName=Alpha}, Alpha=ScalarParameter() {parameterName=Use_Mesh_Alpha})
   B=Clamp(None=Multiply(A=@n26, B=@n34))
  )
 ]
}
